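# Master FX material for mesh particles in the Dealership shared library, combining panning, rotatable textures with noise distortion.
# Particle dynamic parameters drive tiling, offsets, distortion, dissolve and world position offset.
# Color supports gradient lerp, painted-variant tint, hue shift and erosion-edge glow; opacity supports dissolve, fresnel, angle fade, UV-edge masks, depth fade and dithering.
# Camera and normal-based vertex offset uses a quality switch, active at Epic.
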Material /DealershipSharedLibrary/FX/Materials/MasterMaterials/Mesh/M_DLR_Mesh_NoiseDistort
graph {
 expressions {
  MaterialExpressionAbs_0 {class=MaterialExpressionAbs, inputs [None {from=MaterialExpressionDotProduct_7, output=""}], properties {}}
  MaterialExpressionAdd_0 {
   class=MaterialExpressionAdd
   inputs [
    A {from=MaterialExpressionMultiply_3}
    B {from=MaterialExpressionPanner_3}
   ]
  }
  MaterialExpressionAdd_1 {
   class=MaterialExpressionAdd
   inputs [
    A {from=MaterialExpressionMultiply_5}
    B {from=MaterialExpressionAppendVector_0}
   ]
  }
  MaterialExpressionAdd_10 {
   class=MaterialExpressionAdd
   inputs [
    A {from=MaterialExpressionScalarParameter_37}
    B {from=MaterialExpressionNamedRerouteUsage_19}
   ]
  }
  MaterialExpressionAdd_11 {
   class=MaterialExpressionAdd
   inputs [
    A {from=MaterialExpressionNamedRerouteUsage_21}
    B {from=MaterialExpressionIf_1}
   ]
  }
  MaterialExpressionAdd_12 {
   class=MaterialExpressionAdd
   inputs [
    A {from=MaterialExpressionNamedRerouteUsage_4}
    B {from=MaterialExpressionScalarParameter_32}
   ]
  }
  MaterialExpressionAdd_13 {
   class=MaterialExpressionAdd
   inputs [
    A {from=MaterialExpressionMultiply_4}
    B {from=MaterialExpressionAdd_15}
   ]
  }
  MaterialExpressionAdd_14 {
   class=MaterialExpressionAdd
   inputs [
    A {from=MaterialExpressionMultiply_15}
    B {from=MaterialExpressionAppendVector_10}
   ]
  }
  MaterialExpressionAdd_15 {
   class=MaterialExpressionAdd
   inputs [
    A {from=MaterialExpressionPower_0}
    B {from=MaterialExpressionAdd_10}
   ]
  }
  MaterialExpressionAdd_16 {
   class=MaterialExpressionAdd
   inputs [
    A {from=MaterialExpressionAbs_0}
    B {from=MaterialExpressionSubtract_20}
   ]
  }
  MaterialExpressionAdd_17 {
   class=MaterialExpressionAdd
   inputs [
    A {from=MaterialExpressionMultiply_27}
    B {from=MaterialExpressionVectorParameter_9, output=Fresnel Bias}
   ]
  }
  MaterialExpressionAdd_18 {
   class=MaterialExpressionAdd
   inputs [
    A {from=MaterialExpressionMultiply_26}
    B {from=MaterialExpressionVectorParameter_8, output=Fresnel Bias}
   ]
  }
  MaterialExpressionAdd_19 {
   class=MaterialExpressionAdd
   inputs [
    A {from=MaterialExpressionAdd_18}
    B {from=MaterialExpressionAdd_17}
   ]
  }
  MaterialExpressionAdd_2 {
   class=MaterialExpressionAdd
   inputs [
    A {from=MaterialExpressionNamedRerouteUsage_13}
    B {from=MaterialExpressionScalarParameter_29}
   ]
   properties {constB=0.1}
  }
  MaterialExpressionAdd_20 {
   class=MaterialExpressionAdd
   inputs [
    A {from=MaterialExpressionPower_7}
    B {from=MaterialExpressionScalarParameter_46}
   ]
  }
  MaterialExpressionAdd_3 {
   class=MaterialExpressionAdd
   inputs [
    A {from=MaterialExpressionMaterialFunctionCall_1, output=Clamped Camera Offset}
    B {from=MaterialExpressionMultiply_9}
   ]
  }
  MaterialExpressionAdd_4 {
   class=MaterialExpressionAdd
   inputs [
    A {from=MaterialExpressionMultiply_1}
    B {from=MaterialExpressionAppendVector_2}
   ]
  }
  MaterialExpressionAdd_5 {
   class=MaterialExpressionAdd
   inputs [
    A {from=MaterialExpressionDynamicParameter_16, output=Dissolve Amount}
    B {from=MaterialExpressionScalarParameter_4}
   ]
  }
  MaterialExpressionAdd_6 {
   class=MaterialExpressionAdd
   inputs [
    A {from=MaterialExpressionNamedRerouteUsage_8}
    B {from=MaterialExpressionNamedRerouteUsage_18}
   ]
  }
  MaterialExpressionAdd_7 {
   class=MaterialExpressionAdd
   inputs [
    A {from=MaterialExpressionMultiply_14}
    B {from=MaterialExpressionPanner_4}
   ]
  }
  MaterialExpressionAdd_8 {
   class=MaterialExpressionAdd
   inputs [
    A {from=MaterialExpressionNamedRerouteUsage_23}
    B {from=MaterialExpressionAppendVector_11}
   ]
  }
  MaterialExpressionAdd_9 {
   class=MaterialExpressionAdd
   inputs [
    A {from=MaterialExpressionNamedRerouteUsage_24}
    B {from=MaterialExpressionAppendVector_1}
   ]
  }
  MaterialExpressionAppendVector_0 {
   class=MaterialExpressionAppendVector
   inputs [
    A {from=MaterialExpressionDynamicParameter_0, output=Tex1 Offset U}
    B {from=MaterialExpressionDynamicParameter_0, output=Tex1 Offset U}
   ]
  }
  MaterialExpressionAppendVector_1 {
   class=MaterialExpressionAppendVector
   inputs [
    A {from=MaterialExpressionDynamicParameter_0, output=Tex1 Tiling U}
    B {from=MaterialExpressionDynamicParameter_0, output=Tex1 Tiling U}
   ]
  }
  MaterialExpressionAppendVector_10 {
   class=MaterialExpressionAppendVector
   inputs [
    A {from=MaterialExpressionDynamicParameter_8, output=Tex2 Offset U}
    B {from=MaterialExpressionDynamicParameter_8, output=Tex2 Offset U}
   ]
  }
  MaterialExpressionAppendVector_11 {
   class=MaterialExpressionAppendVector
   inputs [
    A {from=MaterialExpressionDynamicParameter_8, output=Tex2 Tiling U}
    B {from=MaterialExpressionDynamicParameter_8, output=Tex2 Tiling U}
   ]
  }
  MaterialExpressionAppendVector_12 {
   class=MaterialExpressionAppendVector
   inputs [
    A {from=MaterialExpressionScalarParameter_16, output=""}
    B {from=MaterialExpressionScalarParameter_17, output=""}
   ]
  }
  MaterialExpressionAppendVector_13 {
   class=MaterialExpressionAppendVector
   inputs [
    A {from=MaterialExpressionScalarParameter_18, output=""}
    B {from=MaterialExpressionScalarParameter_19, output=""}
   ]
  }
  MaterialExpressionAppendVector_14 {
   class=MaterialExpressionAppendVector
   inputs [
    A {from=MaterialExpressionVectorParameter_4, output=ScaleX}
    B {from=MaterialExpressionVectorParameter_4, output=ScaleX}
   ]
  }
  MaterialExpressionAppendVector_15 {
   class=MaterialExpressionAppendVector
   inputs [
    A {from=MaterialExpressionVectorParameter_4, output=SpeedX}
    B {from=MaterialExpressionVectorParameter_4, output=SpeedX}
   ]
  }
  MaterialExpressionAppendVector_16 {
   class=MaterialExpressionAppendVector
   inputs [
    A {from=MaterialExpressionVectorParameter_2, output=SpeedX}
    B {from=MaterialExpressionVectorParameter_2, output=SpeedX}
   ]
  }
  MaterialExpressionAppendVector_2 {
   class=MaterialExpressionAppendVector
   inputs [
    A {from=MaterialExpressionNamedRerouteUsage_5, output=""}
    B {from=MaterialExpressionNamedRerouteUsage_7, output=""}
   ]
  }
  MaterialExpressionAppendVector_3 {
   class=MaterialExpressionAppendVector
   inputs [
    A {from=MaterialExpressionVectorParameter_2, output=ScaleX}
    B {from=MaterialExpressionVectorParameter_2, output=ScaleX}
   ]
  }
  MaterialExpressionAppendVector_4 {
   class=MaterialExpressionAppendVector
   inputs [
    A {from=MaterialExpressionVectorParameter_1, output=Tiling U}
    B {from=MaterialExpressionVectorParameter_1, output=Tiling U}
   ]
  }
  MaterialExpressionAppendVector_5 {
   class=MaterialExpressionAppendVector
   inputs [
    A {from=MaterialExpressionVectorParameter_1, output=Panner U}
    B {from=MaterialExpressionVectorParameter_1, output=Panner U}
   ]
  }
  MaterialExpressionAppendVector_6 {
   class=MaterialExpressionAppendVector
   inputs [
    A {from=MaterialExpressionScalarParameter_12, output=""}
    B {from=MaterialExpressionScalarParameter_13, output=""}
   ]
  }
  MaterialExpressionAppendVector_7 {
   class=MaterialExpressionAppendVector
   inputs [
    A {from=MaterialExpressionScalarParameter_14, output=""}
    B {from=MaterialExpressionScalarParameter_15, output=""}
   ]
  }
  MaterialExpressionAppendVector_8 {
   class=MaterialExpressionAppendVector
   inputs [
    A {from=MaterialExpressionVectorParameter_5, output=Panner U}
    B {from=MaterialExpressionVectorParameter_5, output=Panner U}
   ]
  }
  MaterialExpressionAppendVector_9 {
   class=MaterialExpressionAppendVector
   inputs [
    A {from=MaterialExpressionVectorParameter_5, output=Tiling U}
    B {from=MaterialExpressionVectorParameter_5, output=Tiling U}
   ]
  }
  MaterialExpressionCameraVectorWS_6 {class=MaterialExpressionCameraVectorWS, inputs [], properties {}}
  MaterialExpressionChannelMaskParameter_0 {
   class=MaterialExpressionChannelMaskParameter
   inputs [None {from=MaterialExpressionTextureSampleParameter2D_0, output=RGBA}]
   properties {group=Texture 1, parameterName=Texture 1 - Channel, sortPriority=11}
  }
  MaterialExpressionChannelMaskParameter_1 {
   class=MaterialExpressionChannelMaskParameter
   inputs [None {from=MaterialExpressionTextureSampleParameter2D_3, output=RGBA}]
   properties {group=Texture 2, parameterName=Texture 2 - Channel, sortPriority=11}
  }
  MaterialExpressionChannelMaskParameter_3 {
   class=MaterialExpressionChannelMaskParameter
   inputs [None {from=MaterialExpressionConvert_0, output=""}]
   properties {desc="R=U\r\nG=V\r\nB=1-U\r\nA=1-V", group=Dissolve, parameterName=DissolveWipeMaskDirection, sortPriority=16}
  }
  MaterialExpressionClamp_1 {
   class=MaterialExpressionClamp
   inputs [
    None {from=MaterialExpressionMaterialFunctionCall_4, output=Rotated Values}
    Min {from="", output=""}
    Max {from="", output=""}
   ]
   properties {}
  }
  MaterialExpressionClamp_2 {
   class=MaterialExpressionClamp
   inputs [
    None {from=MaterialExpressionMaterialFunctionCall_3, output=Rotated Values}
    Min {from="", output=""}
    Max {from="", output=""}
   ]
   properties {}
  }
  MaterialExpressionClamp_6 {
   class=MaterialExpressionClamp
   inputs [
    None {from=MaterialExpressionMultiply_35, output=""}
    Min {from="", output=""}
    Max {from="", output=""}
   ]
   properties {}
  }
  MaterialExpressionComponentMask_1 {
   class=MaterialExpressionComponentMask
   inputs [None {from=MaterialExpressionTextureSampleParameter2D_1, output=RGB}]
   properties {g=true, r=true}
  }
  MaterialExpressionComponentMask_2 {class=MaterialExpressionComponentMask, inputs [None {from=MaterialExpressionVectorParameter_7, output=RGB}], properties {g=true, r=true}}
  MaterialExpressionComponentMask_3 {class=MaterialExpressionComponentMask, inputs [None {from=MaterialExpressionTextureCoordinate_8}], properties {g=true}}
  MaterialExpressionComponentMask_4 {class=MaterialExpressionComponentMask, inputs [None {from=MaterialExpressionTextureCoordinate_8}], properties {r=true}}
  MaterialExpressionConstant3Vector_9 {class=MaterialExpressionConstant3Vector, properties {desc=View Direction, constant=<rgba 0 0 -1 0>}}
  MaterialExpressionConstant_0 {class=MaterialExpressionConstant}
  MaterialExpressionConstant_1 {class=MaterialExpressionConstant}
  MaterialExpressionConstant_2 {class=MaterialExpressionConstant}
  MaterialExpressionConstant_29 {class=MaterialExpressionConstant, properties {desc=New Max, r=1}}
  MaterialExpressionConstant_30 {class=MaterialExpressionConstant, properties {desc=New Min}}
  MaterialExpressionConvert_0 {
   class=MaterialExpressionConvert
   inputs [
    None {from=MaterialExpressionTextureCoordinate_5}
    None {from=MaterialExpressionSaturate_0}
   ]
   properties {
    nodeName=Break Float2
    convertInputs [
     {
      type=Vector2
      expressionInput {
       expression=&self:MaterialExpressionTextureCoordinate_5
       inputName=None
       mask=0
       maskA=0
       maskB=0
       maskG=0
       maskR=0
       outputIndex=0
      }
     }
     {
      type=Vector2
      expressionInput {
       expression=&self:MaterialExpressionSaturate_0
       inputName=None
       mask=0
       maskA=0
       maskB=0
       maskG=0
       maskR=0
       outputIndex=0
      }
     }
    ]
    convertOutputs [{type=Vector4, defaultValue=<rgba 0 0 0 1>}]
   }
  }
  MaterialExpressionDivide_6 {
   class=MaterialExpressionDivide
   inputs [
    A {from=MaterialExpressionConstant_29}
    B {from=MaterialExpressionSubtract_18}
   ]
  }
  MaterialExpressionDotProduct_7 {
   class=MaterialExpressionDotProduct
   inputs [
    A {from=MaterialExpressionVertexNormalWS_6, output=""}
    B {from=MaterialExpressionStaticSwitch_2, output=""}
   ]
   properties {}
  }
  MaterialExpressionDynamicParameter_0 {class=MaterialExpressionDynamicParameter, inputs [], properties {paramNames [Tex1 Tiling U, Tex1 Tiling V, Tex1 Offset U, Tex1 Offset V]}}
  MaterialExpressionDynamicParameter_1 {
   class=MaterialExpressionDynamicParameter
   inputs []
   properties {parameterIndex=3, defaultValue=<rgba 0 0 0 0>, paramNames [DissolveSmoothness, DissolveOffsetX, DissolveOffsetY, DissolveMaskOffset]}
  }
  MaterialExpressionDynamicParameter_16 {
   class=MaterialExpressionDynamicParameter
   inputs []
   properties {parameterIndex=2, defaultValue=<rgba 1 0 1 0.05>, paramNames [Texture Multiplier, Dissolve Amount, Distortion Amount, WPO Amount]}
  }
  MaterialExpressionDynamicParameter_8 {
   class=MaterialExpressionDynamicParameter
   inputs []
   properties {parameterIndex=1, paramNames [Tex2 Tiling U, Tex2 Tiling V, Tex2 Offset U, Tex2 Offset V]}
  }
  MaterialExpressionFresnel_3 {
   class=MaterialExpressionFresnel
   inputs [
    ExponentIn {from=MaterialExpressionVectorParameter_8, output=ExponentIn}
    BaseReflectFractionIn {from=MaterialExpressionVectorParameter_8, output=ExponentIn}
    Normal {from="", output=""}
   ]
   properties {}
  }
  MaterialExpressionFresnel_5 {
   class=MaterialExpressionFresnel
   inputs [
    ExponentIn {from=MaterialExpressionVectorParameter_9, output=ExponentIn}
    BaseReflectFractionIn {from=MaterialExpressionVectorParameter_9, output=ExponentIn}
    Normal {from="", output=""}
   ]
   properties {}
  }
  MaterialExpressionIf_0 {
   class=MaterialExpressionIf
   inputs [
    A {from=MaterialExpressionScalarParameter_7, output=""}
    B {from="", output=""}
    "A > B" {from=MaterialExpressionIf_5, output=""}
    "A == B" {from=MaterialExpressionIf_5, output=""}
    "A < B" {from=MaterialExpressionReroute_2, output=""}
   ]
   properties {constB=1}
  }
  MaterialExpressionIf_1 {
   class=MaterialExpressionIf
   inputs [
    A {from=MaterialExpressionScalarParameter_8, output=""}
    B {from="", output=""}
    "A > B" {from=MaterialExpressionAdd_6, output=""}
    "A == B" {from=MaterialExpressionAdd_6, output=""}
    "A < B" {from=MaterialExpressionMultiply_17, output=""}
   ]
   properties {constB=1}
  }
  MaterialExpressionIf_2 {
   class=MaterialExpressionIf
   inputs [
    A {from=MaterialExpressionScalarParameter_9, output=""}
    B {from="", output=""}
    "A > B" {from=MaterialExpressionMultiply_11, output=""}
    "A == B" {from=MaterialExpressionMultiply_11, output=""}
    "A < B" {from=MaterialExpressionMultiply_0, output=""}
   ]
   properties {constB=1}
  }
  MaterialExpressionIf_3 {
   class=MaterialExpressionIf
   inputs [
    A {from=MaterialExpressionScalarParameter_21, output=""}
    B {from="", output=""}
    "A > B" {from=MaterialExpressionAdd_11, output=""}
    "A == B" {from=MaterialExpressionAdd_11, output=""}
    "A < B" {from=MaterialExpressionIf_1, output=""}
   ]
   properties {constB=1}
  }
  MaterialExpressionIf_5 {
   class=MaterialExpressionIf
   inputs [
    A {from=MaterialExpressionScalarParameter_42, output=""}
    B {from="", output=""}
    "A > B" {from=MaterialExpressionSaturate_2, output=""}
    "A == B" {from=MaterialExpressionSaturate_2, output=""}
    "A < B" {from=MaterialExpressionReroute_1, output=""}
   ]
   properties {constB=1}
  }
  MaterialExpressionIf_6 {
   class=MaterialExpressionIf
   inputs [
    A {from=MaterialExpressionScalarParameter_48, output=""}
    B {from="", output=""}
    "A > B" {from=MaterialExpressionMultiply_2, output=""}
    "A == B" {from=MaterialExpressionMultiply_2, output=""}
    "A < B" {from=MaterialExpressionMultiply_22, output=""}
   ]
   properties {constB=1}
  }
  MaterialExpressionIf_7 {
   class=MaterialExpressionIf
   inputs [
    A {from=MaterialExpressionMaterialFunctionCall_10, output=Is Painted}
    B {from="", output=""}
    "A > B" {from=MaterialExpressionConstant_1, output=""}
    "A == B" {from=MaterialExpressionConstant_1, output=""}
    "A < B" {from=MaterialExpressionScalarParameter_51, output=""}
   ]
   properties {constB=1}
  }
  MaterialExpressionIf_8 {
   class=MaterialExpressionIf
   inputs [
    A {from=MaterialExpressionScalarParameter_53, output=""}
    B {from="", output=""}
    "A > B" {from=MaterialExpressionMultiply_36, output=""}
    "A == B" {from=MaterialExpressionIf_6, output=""}
    "A < B" {from=MaterialExpressionIf_6, output=""}
   ]
   properties {}
  }
  MaterialExpressionLinearInterpolate_0 {
   class=MaterialExpressionLinearInterpolate
   inputs [
    A {from=MaterialExpressionMaterialFunctionCall_6, output=Primary RGB}
    B {from=MaterialExpressionVectorParameter_6, output=RGB}
    Alpha {from=MaterialExpressionPower_1}
   ]
  }
  MaterialExpressionLinearInterpolate_1 {
   class=MaterialExpressionLinearInterpolate
   inputs [
    A {from=MaterialExpressionMultiply_16}
    B {from=MaterialExpressionMaterialFunctionCall_9, output=Result}
    Alpha {from=MaterialExpressionNamedRerouteUsage_25}
   ]
  }
  MaterialExpressionLinearInterpolate_16 {
   class=MaterialExpressionLinearInterpolate
   inputs [
    A {from=MaterialExpressionChannelMaskParameter_1, output=R}
    B {from=MaterialExpressionReroute_5}
    Alpha {from=MaterialExpressionScalarParameter_44}
   ]
  }
  MaterialExpressionLinearInterpolate_17 {
   class=MaterialExpressionLinearInterpolate
   inputs [
    A {from=MaterialExpressionChannelMaskParameter_0, output=R}
    B {from=MaterialExpressionTextureSampleParameter2D_0, output=RGB}
    Alpha {from=MaterialExpressionScalarParameter_45}
   ]
  }
  MaterialExpressionLinearInterpolate_2 {
   class=MaterialExpressionLinearInterpolate
   inputs [
    A {from=""}
    B {from=MaterialExpressionScalarParameter_24}
    Alpha {from=MaterialExpressionNamedRerouteUsage_1}
   ]
   properties {constA=1, constB=2}
  }
  MaterialExpressionLinearInterpolate_4 {
   class=MaterialExpressionLinearInterpolate
   inputs [
    A {from=MaterialExpressionScalarParameter_25}
    B {from=MaterialExpressionScalarParameter_26}
    Alpha {from=MaterialExpressionTextureSampleParameter2D_2, output=R}
   ]
   properties {constA=-0.3, constB=0.3}
  }
  MaterialExpressionMaterialFunctionCall_0 {
   class=MaterialExpressionMaterialFunctionCall
   inputs [Emissive {from=MaterialExpressionMultiply_34}]
   properties {materialFunction=&MF_FX_EmissiveAdjustments}
  }
  MaterialExpressionMaterialFunctionCall_1 {
   class=MaterialExpressionMaterialFunctionCall
   inputs [
    Offset Amount {from=MaterialExpressionScalarParameter_0}
    Clamp Padding {from=""}
    World Position {from=""}
   ]
   properties {materialFunction=/Engine/Functions/Engine_MaterialFunctions02/WorldPositionOffset/CameraOffset}
  }
  MaterialExpressionMaterialFunctionCall_10 {
   class=MaterialExpressionMaterialFunctionCall
   inputs [
    Primary Color {from=""}
    Tertiary Color {from=""}
    "Secondary  Color" {from=""}
   ]
   properties {materialFunction=/DealershipSharedLibrary/FX/Materials/MaterialFunctions/MF_DLR_PaintedVariantsData}
  }
  MaterialExpressionMaterialFunctionCall_2 {
   class=MaterialExpressionMaterialFunctionCall
   inputs [
    Opacity In {from=MaterialExpressionIf_8}
    DepthFade Distance {from=MaterialExpressionScalarParameter_1}
   ]
   properties {materialFunction=&MF_VCVFX_DepthFade}
  }
  MaterialExpressionMaterialFunctionCall_3 {
   class=MaterialExpressionMaterialFunctionCall
   inputs [
    UVs {from=MaterialExpressionAdd_7}
    Rotation Center {from=""}
    "Rotation Angle (0-1)" {from=MaterialExpressionScalarParameter_2}
   ]
   properties {materialFunction=&CustomRotator}
  }
  MaterialExpressionMaterialFunctionCall_4 {
   class=MaterialExpressionMaterialFunctionCall
   inputs [
    UVs {from=MaterialExpressionAdd_0}
    Rotation Center {from=""}
    "Rotation Angle (0-1)" {from=MaterialExpressionScalarParameter_3}
   ]
   properties {materialFunction=&CustomRotator}
  }
  MaterialExpressionMaterialFunctionCall_5 {
   class=MaterialExpressionMaterialFunctionCall
   inputs [
    Alpha Threshold {from=MaterialExpressionNamedRerouteUsage_22}
    Random {from=MaterialExpressionScalarParameter_5}
   ]
   properties {materialFunction=&DitherTemporalAA}
  }
  MaterialExpressionMaterialFunctionCall_6 {
   class=MaterialExpressionMaterialFunctionCall
   inputs [
    Primary Color {from=MaterialExpressionVectorParameter_3, output=RGBA}
    Tertiary Color {from=""}
    "Secondary  Color" {from=""}
   ]
   properties {materialFunction=&MF_DLR_PaintedVariantsData}
  }
  MaterialExpressionMaterialFunctionCall_7 {
   class=MaterialExpressionMaterialFunctionCall
   inputs [
    Erosion {from=MaterialExpressionNamedRerouteUsage_17}
    Mask {from=MaterialExpressionIf_0}
    Edge Width {from=MaterialExpressionScalarParameter_23}
    Width {from=MaterialExpressionAdd_12}
   ]
   properties {materialFunction=/DealershipSharedLibrary/FX/Materials/MaterialFunctions/MF_DLR_Erode}
  }
  MaterialExpressionMaterialFunctionCall_9 {
   class=MaterialExpressionMaterialFunctionCall
   inputs [
    Hue Shift Percentage {from=MaterialExpressionIf_7}
    Texture {from=MaterialExpressionIf_2}
   ]
   properties {materialFunction=&HueShift}
  }
  MaterialExpressionMultiply_0 {
   class=MaterialExpressionMultiply
   inputs [
    A {from=MaterialExpressionParticleColor_0, output=RGB}
    B {from=MaterialExpressionVectorParameter_0, output=RGB}
   ]
  }
  MaterialExpressionMultiply_1 {
   class=MaterialExpressionMultiply
   inputs [
    A {from=MaterialExpressionTextureCoordinate_4}
    B {from=MaterialExpressionAppendVector_3}
   ]
  }
  MaterialExpressionMultiply_10 {
   class=MaterialExpressionMultiply
   inputs [
    A {from=MaterialExpressionTextureCoordinate_3}
    B {from=MaterialExpressionAppendVector_4}
   ]
   properties {constB=2}
  }
  MaterialExpressionMultiply_11 {
   class=MaterialExpressionMultiply
   inputs [
    A {from=MaterialExpressionPower_6}
    B {from=MaterialExpressionScalarParameter_20}
   ]
  }
  MaterialExpressionMultiply_12 {
   class=MaterialExpressionMultiply
   inputs [
    A {from=MaterialExpressionTextureCoordinate_7}
    B {from=MaterialExpressionAppendVector_14}
   ]
  }
  MaterialExpressionMultiply_13 {
   class=MaterialExpressionMultiply
   inputs [
    A {from=MaterialExpressionPower_2}
    B {from=MaterialExpressionScalarParameter_11}
   ]
  }
  MaterialExpressionMultiply_14 {
   class=MaterialExpressionMultiply
   inputs [
    A {from=MaterialExpressionNamedRerouteUsage_9}
    B {from=MaterialExpressionMultiply_24}
   ]
  }
  MaterialExpressionMultiply_15 {
   class=MaterialExpressionMultiply
   inputs [
    A {from=MaterialExpressionTextureCoordinate_2}
    B {from=MaterialExpressionAdd_8}
   ]
  }
  MaterialExpressionMultiply_16 {
   class=MaterialExpressionMultiply
   inputs [
    A {from=MaterialExpressionVectorParameter_10, output=RGB}
    B {from=MaterialExpressionScalarParameter_27}
   ]
  }
  MaterialExpressionMultiply_17 {
   class=MaterialExpressionMultiply
   inputs [
    A {from=MaterialExpressionNamedRerouteUsage_8}
    B {from=MaterialExpressionNamedRerouteUsage_18}
   ]
  }
  MaterialExpressionMultiply_18 {
   class=MaterialExpressionMultiply
   inputs [
    A {from=MaterialExpressionSaturate_5}
    B {from=MaterialExpressionNamedRerouteUsage_12}
   ]
  }
  MaterialExpressionMultiply_19 {
   class=MaterialExpressionMultiply
   inputs [
    A {from=MaterialExpressionLinearInterpolate_4}
    B {from=MaterialExpressionNamedRerouteUsage_14}
   ]
  }
  MaterialExpressionMultiply_2 {
   class=MaterialExpressionMultiply
   inputs [
    A {from=MaterialExpressionParticleColor_1, output=A}
    B {from=MaterialExpressionMultiply_22}
   ]
  }
  MaterialExpressionMultiply_20 {
   class=MaterialExpressionMultiply
   inputs [
    A {from=MaterialExpressionComponentMask_4}
    B {from=MaterialExpressionOneMinus_0}
   ]
  }
  MaterialExpressionMultiply_21 {
   class=MaterialExpressionMultiply
   inputs [
    A {from=MaterialExpressionMultiply_20}
    B {from=""}
   ]
   properties {constB=4}
  }
  MaterialExpressionMultiply_22 {
   class=MaterialExpressionMultiply
   inputs [
    A {from=MaterialExpressionMultiply_30}
    B {from=MaterialExpressionNamedRerouteUsage_3}
   ]
  }
  MaterialExpressionMultiply_23 {
   class=MaterialExpressionMultiply
   inputs [
    A {from=MaterialExpressionNamedRerouteUsage_15}
    B {from=MaterialExpressionScalarParameter_40}
   ]
  }
  MaterialExpressionMultiply_24 {
   class=MaterialExpressionMultiply
   inputs [
    A {from=MaterialExpressionNamedRerouteUsage_16}
    B {from=MaterialExpressionScalarParameter_31}
   ]
  }
  MaterialExpressionMultiply_25 {
   class=MaterialExpressionMultiply
   inputs [
    A {from=MaterialExpressionMultiply_31}
    B {from=MaterialExpressionScalarParameter_47}
   ]
  }
  MaterialExpressionMultiply_26 {
   class=MaterialExpressionMultiply
   inputs [
    A {from=MaterialExpressionSaturate_1}
    B {from=MaterialExpressionReroute_8}
   ]
  }
  MaterialExpressionMultiply_27 {
   class=MaterialExpressionMultiply
   inputs [
    A {from=MaterialExpressionSaturate_8}
    B {from=MaterialExpressionReroute_4}
   ]
  }
  MaterialExpressionMultiply_28 {
   class=MaterialExpressionMultiply
   inputs [
    A {from=MaterialExpressionPower_4}
    B {from=MaterialExpressionPower_5}
   ]
  }
  MaterialExpressionMultiply_29 {
   class=MaterialExpressionMultiply
   inputs [
    A {from=MaterialExpressionMultiply_28}
    B {from=MaterialExpressionPower_3}
   ]
  }
  MaterialExpressionMultiply_3 {
   class=MaterialExpressionMultiply
   inputs [
    A {from=MaterialExpressionNamedRerouteUsage_10}
    B {from=MaterialExpressionMultiply_23}
   ]
  }
  MaterialExpressionMultiply_30 {
   class=MaterialExpressionMultiply
   inputs [
    A {from=MaterialExpressionNamedRerouteUsage_6}
    B {from=MaterialExpressionNamedRerouteUsage_11}
   ]
  }
  MaterialExpressionMultiply_31 {
   class=MaterialExpressionMultiply
   inputs [
    A {from=MaterialExpressionMaterialFunctionCall_2, output=Result}
    B {from=MaterialExpressionMultiply_29}
   ]
  }
  MaterialExpressionMultiply_32 {
   class=MaterialExpressionMultiply
   inputs [
    A {from=MaterialExpressionLinearInterpolate_16}
    B {from=MaterialExpressionScalarParameter_30}
   ]
  }
  MaterialExpressionMultiply_33 {
   class=MaterialExpressionMultiply
   inputs [
    A {from=MaterialExpressionLinearInterpolate_17}
    B {from=MaterialExpressionScalarParameter_28}
   ]
  }
  MaterialExpressionMultiply_34 {
   class=MaterialExpressionMultiply
   inputs [
    A {from=MaterialExpressionLinearInterpolate_1}
    B {from=MaterialExpressionScalarParameter_49}
   ]
  }
  MaterialExpressionMultiply_35 {
   class=MaterialExpressionMultiply
   inputs [
    A {from=MaterialExpressionDivide_6}
    B {from=MaterialExpressionAdd_16}
   ]
  }
  MaterialExpressionMultiply_36 {
   class=MaterialExpressionMultiply
   inputs [
    A {from=MaterialExpressionIf_6}
    B {from=MaterialExpressionNamedRerouteUsage_20}
   ]
  }
  MaterialExpressionMultiply_4 {
   class=MaterialExpressionMultiply
   inputs [
    A {from=MaterialExpressionSaturate_10}
    B {from=MaterialExpressionAdd_15}
   ]
  }
  MaterialExpressionMultiply_5 {
   class=MaterialExpressionMultiply
   inputs [
    A {from=MaterialExpressionTextureCoordinate_0}
    B {from=MaterialExpressionAdd_9}
   ]
  }
  MaterialExpressionMultiply_6 {
   class=MaterialExpressionMultiply
   inputs [
    A {from=MaterialExpressionTextureCoordinate_1}
    B {from=MaterialExpressionAppendVector_9}
   ]
  }
  MaterialExpressionMultiply_7 {
   class=MaterialExpressionMultiply
   inputs [
    A {from=MaterialExpressionTextureCoordinate_6}
    B {from=MaterialExpressionComponentMask_2}
   ]
  }
  MaterialExpressionMultiply_8 {
   class=MaterialExpressionMultiply
   inputs [
    A {from=MaterialExpressionLinearInterpolate_0}
    B {from=MaterialExpressionAdd_2}
   ]
  }
  MaterialExpressionMultiply_9 {
   class=MaterialExpressionMultiply
   inputs [
    A {from=MaterialExpressionVertexNormalWS_0}
    B {from=MaterialExpressionMultiply_19}
   ]
  }
  MaterialExpressionNamedRerouteDeclaration_0 {
   class=MaterialExpressionNamedRerouteDeclaration
   inputs [None {from=MaterialExpressionStaticSwitchParameter_8}]
   properties {name=WPO_Out, nodeColor=<rgba 1 0.9098 0.2641 1>}
  }
  MaterialExpressionNamedRerouteDeclaration_1 {
   class=MaterialExpressionNamedRerouteDeclaration
   inputs [None {from=MaterialExpressionMultiply_25}]
   properties {name=Opacity Out, nodeColor=<rgba 0 1 0.9531 1>}
  }
  MaterialExpressionNamedRerouteDeclaration_10 {
   class=MaterialExpressionNamedRerouteDeclaration
   inputs [None {from=MaterialExpressionDynamicParameter_16, output=Distortion Amount}]
   properties {name=DP_Distortion Amount, nodeColor=<rgba 1 0.0947 0.1282 1>}
  }
  MaterialExpressionNamedRerouteDeclaration_11 {
   class=MaterialExpressionNamedRerouteDeclaration
   inputs [None {from=MaterialExpressionDynamicParameter_16, output=WPO Amount}]
   properties {name=DP_WPO Multiplier, nodeColor=<rgba 1 0.0947 0.1282 1>}
  }
  MaterialExpressionNamedRerouteDeclaration_12 {
   class=MaterialExpressionNamedRerouteDeclaration
   inputs [None {from=MaterialExpressionDynamicParameter_1, output=DissolveOffsetY}]
   properties {name=DP_DissolveOffsetY, nodeColor=<rgba 1 0.0947 0.1282 1>}
  }
  MaterialExpressionNamedRerouteDeclaration_13 {
   class=MaterialExpressionNamedRerouteDeclaration
   inputs [None {from=MaterialExpressionDynamicParameter_1, output=DissolveMaskOffset}]
   properties {name=DP_DissolveMaskOffset, nodeColor=<rgba 1 0.0947 0.1282 1>}
  }
  MaterialExpressionNamedRerouteDeclaration_14 {
   class=MaterialExpressionNamedRerouteDeclaration
   inputs [None {from=MaterialExpressionAppendVector_6}]
   properties {name=TilingText2, nodeColor=<rgba 1 0 0.3516 1>}
  }
  MaterialExpressionNamedRerouteDeclaration_15 {
   class=MaterialExpressionNamedRerouteDeclaration
   inputs [None {from=MaterialExpressionAppendVector_12}]
   properties {name=TilingText1, nodeColor=<rgba 1 0 0.3516 1>}
  }
  MaterialExpressionNamedRerouteDeclaration_16 {
   class=MaterialExpressionNamedRerouteDeclaration
   inputs [None {from=MaterialExpressionMultiply_13}]
   properties {name=Texture 3 Out, nodeColor=<rgba 0 0.7187 1 1>}
  }
  MaterialExpressionNamedRerouteDeclaration_17 {
   class=MaterialExpressionNamedRerouteDeclaration
   inputs [None {from=MaterialExpressionAdd_20}]
   properties {name=Texture 1 Out, nodeColor=<rgba 1 0.9609 0 1>}
  }
  MaterialExpressionNamedRerouteDeclaration_18 {
   class=MaterialExpressionNamedRerouteDeclaration
   inputs [None {from=MaterialExpressionPower_8}]
   properties {name=Texture 2 Out, nodeColor=<rgba 0.6406 0 1 1>}
  }
  MaterialExpressionNamedRerouteDeclaration_19 {
   class=MaterialExpressionNamedRerouteDeclaration
   inputs [None {from=MaterialExpressionMaterialFunctionCall_7, output=Edge}]
   properties {name=DissolveEdgeOut, nodeColor=<rgba 0 1 0.5078 1>}
  }
  MaterialExpressionNamedRerouteDeclaration_2 {
   class=MaterialExpressionNamedRerouteDeclaration
   inputs [None {from=MaterialExpressionMaterialFunctionCall_7, output=Result}]
   properties {name=Dissolve Out, nodeColor=<rgba 0.9207 0.4514 1 1>}
  }
  MaterialExpressionNamedRerouteDeclaration_20 {
   class=MaterialExpressionNamedRerouteDeclaration
   inputs [None {from=MaterialExpressionClamp_6}]
   properties {name=AngleFade, nodeColor=<rgba 0.4816 0 1 1>}
  }
  MaterialExpressionNamedRerouteDeclaration_3 {
   class=MaterialExpressionNamedRerouteDeclaration
   inputs [None {from=MaterialExpressionDynamicParameter_1, output=DissolveSmoothness}]
   properties {name=DP_DissolveSmoothness, nodeColor=<rgba 1 0.0947 0.1282 1>}
  }
  MaterialExpressionNamedRerouteDeclaration_4 {
   class=MaterialExpressionNamedRerouteDeclaration
   inputs [None {from=MaterialExpressionMultiply_18}]
   properties {name=Base Texture, nodeColor=<rgba 0.0078 1 0 1>}
  }
  MaterialExpressionNamedRerouteDeclaration_5 {
   class=MaterialExpressionNamedRerouteDeclaration
   inputs [None {from=MaterialExpressionDynamicParameter_1, output=DissolveOffsetX}]
   properties {name=DP_DissolveOffsetX, nodeColor=<rgba 1 0.0947 0.1282 1>}
  }
  MaterialExpressionNamedRerouteDeclaration_6 {
   class=MaterialExpressionNamedRerouteDeclaration
   inputs [None {from=MaterialExpressionSubtract_2}]
   properties {name=Texture Distortion Out, nodeColor=<rgba 0.5435 1 0.2151 1>}
  }
  MaterialExpressionNamedRerouteDeclaration_7 {
   class=MaterialExpressionNamedRerouteDeclaration
   inputs [None {from=MaterialExpressionAdd_19}]
   properties {name=Fresnel Out, nodeColor=<rgba 0.1892 0.1821 1 1>}
  }
  MaterialExpressionNamedRerouteDeclaration_8 {
   class=MaterialExpressionNamedRerouteDeclaration
   inputs [None {from=MaterialExpressionDynamicParameter_16, output=Texture Multiplier}]
   properties {name=DP_Texture Multiplier, nodeColor=<rgba 1 0.0947 0.1282 1>}
  }
  MaterialExpressionNamedRerouteDeclaration_9 {
   class=MaterialExpressionNamedRerouteDeclaration
   inputs [None {from=MaterialExpressionAdd_5}]
   properties {name=DP_Dissolve Amount, nodeColor=<rgba 1 0.0947 0.1282 1>}
  }
  MaterialExpressionNamedRerouteUsage_0 {class=MaterialExpressionNamedRerouteUsage}
  MaterialExpressionNamedRerouteUsage_1 {class=MaterialExpressionNamedRerouteUsage}
  MaterialExpressionNamedRerouteUsage_10 {class=MaterialExpressionNamedRerouteUsage}
  MaterialExpressionNamedRerouteUsage_11 {class=MaterialExpressionNamedRerouteUsage}
  MaterialExpressionNamedRerouteUsage_12 {class=MaterialExpressionNamedRerouteUsage}
  MaterialExpressionNamedRerouteUsage_13 {class=MaterialExpressionNamedRerouteUsage}
  MaterialExpressionNamedRerouteUsage_14 {class=MaterialExpressionNamedRerouteUsage}
  MaterialExpressionNamedRerouteUsage_15 {class=MaterialExpressionNamedRerouteUsage}
  MaterialExpressionNamedRerouteUsage_16 {class=MaterialExpressionNamedRerouteUsage}
  MaterialExpressionNamedRerouteUsage_17 {class=MaterialExpressionNamedRerouteUsage}
  MaterialExpressionNamedRerouteUsage_18 {class=MaterialExpressionNamedRerouteUsage}
  MaterialExpressionNamedRerouteUsage_19 {class=MaterialExpressionNamedRerouteUsage}
  MaterialExpressionNamedRerouteUsage_2 {class=MaterialExpressionNamedRerouteUsage}
  MaterialExpressionNamedRerouteUsage_20 {class=MaterialExpressionNamedRerouteUsage}
  MaterialExpressionNamedRerouteUsage_21 {class=MaterialExpressionNamedRerouteUsage}
  MaterialExpressionNamedRerouteUsage_22 {class=MaterialExpressionNamedRerouteUsage}
  MaterialExpressionNamedRerouteUsage_23 {class=MaterialExpressionNamedRerouteUsage}
  MaterialExpressionNamedRerouteUsage_24 {class=MaterialExpressionNamedRerouteUsage}
  MaterialExpressionNamedRerouteUsage_25 {class=MaterialExpressionNamedRerouteUsage}
  MaterialExpressionNamedRerouteUsage_3 {class=MaterialExpressionNamedRerouteUsage}
  MaterialExpressionNamedRerouteUsage_4 {class=MaterialExpressionNamedRerouteUsage}
  MaterialExpressionNamedRerouteUsage_5 {class=MaterialExpressionNamedRerouteUsage}
  MaterialExpressionNamedRerouteUsage_6 {class=MaterialExpressionNamedRerouteUsage}
  MaterialExpressionNamedRerouteUsage_7 {class=MaterialExpressionNamedRerouteUsage}
  MaterialExpressionNamedRerouteUsage_8 {class=MaterialExpressionNamedRerouteUsage}
  MaterialExpressionNamedRerouteUsage_9 {class=MaterialExpressionNamedRerouteUsage}
  MaterialExpressionOneMinus_0 {class=MaterialExpressionOneMinus, inputs [None {from=MaterialExpressionComponentMask_4}]}
  MaterialExpressionOneMinus_1 {class=MaterialExpressionOneMinus, inputs [None {from=MaterialExpressionComponentMask_3}]}
  MaterialExpressionOneMinus_2 {class=MaterialExpressionOneMinus, inputs [None {from=MaterialExpressionTextureCoordinate_5}]}
  MaterialExpressionOneMinus_4 {class=MaterialExpressionOneMinus, inputs [None {from=MaterialExpressionFresnel_5}]}
  MaterialExpressionPanner_0 {
   class=MaterialExpressionPanner
   inputs [
    Coordinate {from=MaterialExpressionMultiply_10}
    Time {from=""}
    Speed {from=MaterialExpressionAppendVector_5}
   ]
   properties {speedX=0.2, speedY=0.15}
  }
  MaterialExpressionPanner_1 {
   class=MaterialExpressionPanner
   inputs [
    Coordinate {from=MaterialExpressionAdd_4}
    Time {from=""}
    Speed {from=MaterialExpressionAppendVector_16}
   ]
  }
  MaterialExpressionPanner_2 {
   class=MaterialExpressionPanner
   inputs [
    Coordinate {from=MaterialExpressionMultiply_6}
    Time {from=""}
    Speed {from=MaterialExpressionAppendVector_8}
   ]
  }
  MaterialExpressionPanner_3 {
   class=MaterialExpressionPanner
   inputs [
    Coordinate {from=MaterialExpressionAdd_1}
    Time {from=""}
    Speed {from=MaterialExpressionAppendVector_13}
   ]
  }
  MaterialExpressionPanner_4 {
   class=MaterialExpressionPanner
   inputs [
    Coordinate {from=MaterialExpressionAdd_14}
    Time {from=""}
    Speed {from=MaterialExpressionAppendVector_7}
   ]
  }
  MaterialExpressionPanner_5 {
   class=MaterialExpressionPanner
   inputs [
    Coordinate {from=MaterialExpressionMultiply_12}
    Time {from=""}
    Speed {from=MaterialExpressionAppendVector_15}
   ]
  }
  MaterialExpressionParticleColor_0 {class=MaterialExpressionParticleColor, inputs [], properties {}}
  MaterialExpressionParticleColor_1 {class=MaterialExpressionParticleColor, inputs [], properties {}}
  MaterialExpressionPower_0 {
   class=MaterialExpressionPower
   inputs [
    Base {from=MaterialExpressionChannelMaskParameter_3, output=R}
    Exp {from=MaterialExpressionScalarParameter_6}
   ]
  }
  MaterialExpressionPower_1 {
   class=MaterialExpressionPower
   inputs [
    Base {from=MaterialExpressionTextureSampleParameter2D_5, output=R}
    Exp {from=MaterialExpressionScalarParameter_10}
   ]
  }
  MaterialExpressionPower_2 {
   class=MaterialExpressionPower
   inputs [
    Base {from=MaterialExpressionTextureSampleParameter2D_6, output=RGB}
    Exp {from=MaterialExpressionScalarParameter_22}
   ]
  }
  MaterialExpressionPower_3 {
   class=MaterialExpressionPower
   inputs [
    Base {from=MaterialExpressionMultiply_21}
    Exp {from=MaterialExpressionScalarParameter_33}
   ]
  }
  MaterialExpressionPower_4 {
   class=MaterialExpressionPower
   inputs [
    Base {from=MaterialExpressionComponentMask_3}
    Exp {from=MaterialExpressionScalarParameter_34}
   ]
  }
  MaterialExpressionPower_5 {
   class=MaterialExpressionPower
   inputs [
    Base {from=MaterialExpressionOneMinus_1}
    Exp {from=MaterialExpressionScalarParameter_35}
   ]
  }
  MaterialExpressionPower_6 {
   class=MaterialExpressionPower
   inputs [
    Base {from=MaterialExpressionMultiply_8}
    Exp {from=MaterialExpressionScalarParameter_50}
   ]
  }
  MaterialExpressionPower_7 {
   class=MaterialExpressionPower
   inputs [
    Base {from=MaterialExpressionMultiply_32}
    Exp {from=MaterialExpressionScalarParameter_38}
   ]
  }
  MaterialExpressionPower_8 {
   class=MaterialExpressionPower
   inputs [
    Base {from=MaterialExpressionMultiply_33}
    Exp {from=MaterialExpressionScalarParameter_39}
   ]
  }
  MaterialExpressionQualitySwitch_2 {
   class=MaterialExpressionQualitySwitch
   inputs [
    Default {from=MaterialExpressionConstant_2, output=""}
    Low {from=MaterialExpressionConstant_2, output=""}
    High {from="", output=""}
    Medium {from=MaterialExpressionConstant_2, output=""}
    Epic {from=MaterialExpressionAdd_3, output=""}
   ]
   properties {}
  }
  MaterialExpressionReroute_1 {class=MaterialExpressionReroute, inputs [None {from=MaterialExpressionMultiply_4}]}
  MaterialExpressionReroute_2 {class=MaterialExpressionReroute, inputs [None {from=MaterialExpressionSaturate_10}]}
  MaterialExpressionReroute_4 {class=MaterialExpressionReroute, inputs [None {from=MaterialExpressionVectorParameter_9, output=Fresnel Amount}]}
  MaterialExpressionReroute_5 {class=MaterialExpressionReroute, inputs [None {from=MaterialExpressionTextureSampleParameter2D_3, output=RGB}]}
  MaterialExpressionReroute_8 {class=MaterialExpressionReroute, inputs [None {from=MaterialExpressionVectorParameter_8, output=Fresnel Amount}]}
  MaterialExpressionSaturate_0 {class=MaterialExpressionSaturate, inputs [None {from=MaterialExpressionOneMinus_2}]}
  MaterialExpressionSaturate_1 {class=MaterialExpressionSaturate, inputs [None {from=MaterialExpressionFresnel_3}]}
  MaterialExpressionSaturate_10 {class=MaterialExpressionSaturate, inputs [None {from=MaterialExpressionTextureSampleParameter2D_4, output=R}]}
  MaterialExpressionSaturate_2 {class=MaterialExpressionSaturate, inputs [None {from=MaterialExpressionAdd_13}]}
  MaterialExpressionSaturate_5 {class=MaterialExpressionSaturate, inputs [None {from=MaterialExpressionIf_3}]}
  MaterialExpressionSaturate_8 {class=MaterialExpressionSaturate, inputs [None {from=MaterialExpressionOneMinus_4}]}
  MaterialExpressionScalarParameter_0 {class=MaterialExpressionScalarParameter, properties {group=WPO, parameterName=WPO - CameraOffset}}
  MaterialExpressionScalarParameter_1 {class=MaterialExpressionScalarParameter, properties {defaultValue=100, group=Opacity, parameterName=Depth Fade Distance}}
  MaterialExpressionScalarParameter_10 {class=MaterialExpressionScalarParameter, properties {defaultValue=0.35, group=Color, parameterName=LerpColorNoiseMaskPower, sortPriority=20}}
  MaterialExpressionScalarParameter_11 {class=MaterialExpressionScalarParameter, properties {defaultValue=1, group=Texture 3, parameterName=Texture 3- Multiplier}}
  MaterialExpressionScalarParameter_12 {class=MaterialExpressionScalarParameter, properties {group=Texture 2, parameterName=Texture 2 - TilingU}}
  MaterialExpressionScalarParameter_13 {class=MaterialExpressionScalarParameter, properties {group=Texture 2, parameterName=Texture 2 - TilingV}}
  MaterialExpressionScalarParameter_14 {class=MaterialExpressionScalarParameter, properties {group=Texture 2, parameterName=Texture 2 - SpeedX}}
  MaterialExpressionScalarParameter_15 {class=MaterialExpressionScalarParameter, properties {group=Texture 2, parameterName=Texture 2 - SpeedY}}
  MaterialExpressionScalarParameter_16 {class=MaterialExpressionScalarParameter, properties {group=Texture 1, parameterName=Texture 1 - TilingU}}
  MaterialExpressionScalarParameter_17 {class=MaterialExpressionScalarParameter, properties {group=Texture 1, parameterName=Texture 1 - TilingV}}
  MaterialExpressionScalarParameter_18 {class=MaterialExpressionScalarParameter, properties {group=Texture 1, parameterName=Texture 1 - SpeedX}}
  MaterialExpressionScalarParameter_19 {class=MaterialExpressionScalarParameter, properties {group=Texture 1, parameterName=Texture 1 - SpeedY}}
  MaterialExpressionScalarParameter_2 {class=MaterialExpressionScalarParameter, properties {group=Texture 2, parameterName=Texture 2 - Rotator}}
  MaterialExpressionScalarParameter_20 {class=MaterialExpressionScalarParameter, properties {defaultValue=1, group=Color, parameterName=EmissiveMultiplier, sortPriority=9}}
  MaterialExpressionScalarParameter_21 {class=MaterialExpressionScalarParameter, properties {group=Texture 3, parameterName=AddTexture3?}}
  MaterialExpressionScalarParameter_22 {class=MaterialExpressionScalarParameter, properties {defaultValue=1, group=Texture 3, parameterName=Texture 3- Power}}
  MaterialExpressionScalarParameter_23 {class=MaterialExpressionScalarParameter, properties {group=Dissolve, parameterName=DissolveEdgeWidth, sortPriority=28}}
  MaterialExpressionScalarParameter_24 {class=MaterialExpressionScalarParameter, properties {defaultValue=1, group=Refraction, parameterName=Refraction - Amount}}
  MaterialExpressionScalarParameter_25 {class=MaterialExpressionScalarParameter, properties {defaultValue=-200, group=WPO, parameterName=WPO - Min Amount}}
  MaterialExpressionScalarParameter_26 {class=MaterialExpressionScalarParameter, properties {defaultValue=400, group=WPO, parameterName="WPO -Max  Amount"}}
  MaterialExpressionScalarParameter_27 {class=MaterialExpressionScalarParameter, properties {defaultValue=1, group=Dissolve, parameterName=Dissolve Highlight Gain}}
  MaterialExpressionScalarParameter_28 {class=MaterialExpressionScalarParameter, properties {defaultValue=1, group=Texture 1, parameterName=Texture 1 - Multiplier}}
  MaterialExpressionScalarParameter_29 {class=MaterialExpressionScalarParameter, properties {group=Color, parameterName=AddToBaseTextureValue, sortPriority=11}}
  MaterialExpressionScalarParameter_3 {class=MaterialExpressionScalarParameter, properties {group=Texture 1, parameterName=Texture 1 - Rotator}}
  MaterialExpressionScalarParameter_30 {class=MaterialExpressionScalarParameter, properties {defaultValue=2, group=Texture 2, parameterName=Texture 2 - Multiplier}}
  MaterialExpressionScalarParameter_31 {class=MaterialExpressionScalarParameter, properties {defaultValue=0.5, group=Texture 2, parameterName=Texture 2 - Distortion Amount}}
  MaterialExpressionScalarParameter_32 {class=MaterialExpressionScalarParameter, properties {group=Dissolve, parameterName=DissolveSmoothness}}
  MaterialExpressionScalarParameter_33 {class=MaterialExpressionScalarParameter, properties {group=Opacity, parameterName=XEdgeOpacityMask}}
  MaterialExpressionScalarParameter_34 {class=MaterialExpressionScalarParameter, properties {group=Opacity, parameterName=YUpEdgeOpacityMaskUp}}
  MaterialExpressionScalarParameter_35 {class=MaterialExpressionScalarParameter, properties {group=Opacity, parameterName=YDownEdgeOpacityMaskPower}}
  MaterialExpressionScalarParameter_36 {class=MaterialExpressionScalarParameter, properties {defaultValue=0.5, parameterName=Angle Fade - Opaque Facing Ratio}}
  MaterialExpressionScalarParameter_37 {class=MaterialExpressionScalarParameter, properties {group=Dissolve, parameterName=AnimateDissolveMask, sortPriority=20}}
  MaterialExpressionScalarParameter_38 {class=MaterialExpressionScalarParameter, properties {defaultValue=1, group=Texture 2, parameterName=Texture 2 - Power}}
  MaterialExpressionScalarParameter_39 {class=MaterialExpressionScalarParameter, properties {defaultValue=1, group=Texture 1, parameterName=Texture 1 - Power}}
  MaterialExpressionScalarParameter_4 {class=MaterialExpressionScalarParameter, properties {group=Dissolve, parameterName=MaterialDissolve}}
  MaterialExpressionScalarParameter_40 {class=MaterialExpressionScalarParameter, properties {defaultValue=1, group=Texture 1, parameterName=Texture 1 - Distortion Amount}}
  MaterialExpressionScalarParameter_42 {class=MaterialExpressionScalarParameter, properties {group=Dissolve, parameterName=MultiplyAddDissolveMask, sortPriority=13}}
  MaterialExpressionScalarParameter_44 {class=MaterialExpressionScalarParameter, properties {group=Texture 2, parameterName=Texture 2 - Use Channel}}
  MaterialExpressionScalarParameter_45 {class=MaterialExpressionScalarParameter, properties {group=Texture 1, parameterName=Texture 1 - Use Channel}}
  MaterialExpressionScalarParameter_46 {class=MaterialExpressionScalarParameter, properties {group=Texture 2, parameterName=Texture 2 - Bias}}
  MaterialExpressionScalarParameter_47 {class=MaterialExpressionScalarParameter, properties {defaultValue=1, group=Opacity, parameterName=Final Opacity Adjust}}
  MaterialExpressionScalarParameter_48 {class=MaterialExpressionScalarParameter, properties {defaultValue=1, group=Opacity, parameterName=Use Particle Alpha}}
  MaterialExpressionScalarParameter_49 {class=MaterialExpressionScalarParameter, properties {defaultValue=1, group=Color, parameterName=Emissive - Gain}}
  MaterialExpressionScalarParameter_5 {class=MaterialExpressionScalarParameter, properties {group=Opacity, parameterName=MaskedDither}}
  MaterialExpressionScalarParameter_50 {class=MaterialExpressionScalarParameter, properties {defaultValue=1, group=Color, parameterName=Emissive - Power, sortPriority=9}}
  MaterialExpressionScalarParameter_51 {class=MaterialExpressionScalarParameter, properties {group=Color, parameterName=HueShiftPercentage}}
  MaterialExpressionScalarParameter_52 {class=MaterialExpressionScalarParameter, properties {defaultValue=0.5, parameterName=Angle Fade - Transparent Facing Ratio}}
  MaterialExpressionScalarParameter_53 {class=MaterialExpressionScalarParameter, properties {parameterName=UseAngleFade}}
  MaterialExpressionScalarParameter_6 {class=MaterialExpressionScalarParameter, properties {defaultValue=1, group=Dissolve, parameterName=DissolveMaskPower, sortPriority=24}}
  MaterialExpressionScalarParameter_7 {class=MaterialExpressionScalarParameter, properties {group=Dissolve, parameterName=UseDissolveWipeMask, sortPriority=12}}
  MaterialExpressionScalarParameter_8 {class=MaterialExpressionScalarParameter, properties {parameterName=AddBaseTextures?}}
  MaterialExpressionScalarParameter_9 {class=MaterialExpressionScalarParameter, properties {group=Color, parameterName=UseMaskForColor?, sortPriority=8}}
  MaterialExpressionStaticBoolParameter_0 {class=MaterialExpressionStaticBoolParameter, inputs [], properties {group=Texture 1, parameterName=Texture 1 - Clamp?}}
  MaterialExpressionStaticBoolParameter_1 {class=MaterialExpressionStaticBoolParameter, inputs [], properties {group=Texture 2, parameterName=Texture 2 - Clamp?}}
  MaterialExpressionStaticBool_1 {class=MaterialExpressionStaticBool, properties {value=true}}
  MaterialExpressionStaticSwitchParameter_8 {
   class=MaterialExpressionStaticSwitchParameter
   inputs [
    True {from=MaterialExpressionQualitySwitch_2, output=""}
    False {from=MaterialExpressionConstant_0, output=""}
   ]
   properties {defaultValue=true, group=WPO, parameterName=Use WPO?}
  }
  MaterialExpressionStaticSwitch_0 {
   class=MaterialExpressionStaticSwitch
   inputs [
    True {from=MaterialExpressionClamp_1, output=""}
    False {from=MaterialExpressionMaterialFunctionCall_4, output=Rotated Values}
    Value {from=MaterialExpressionStaticBoolParameter_0, output=""}
   ]
   properties {}
  }
  MaterialExpressionStaticSwitch_1 {
   class=MaterialExpressionStaticSwitch
   inputs [
    True {from=MaterialExpressionClamp_2, output=""}
    False {from=MaterialExpressionMaterialFunctionCall_3, output=Rotated Values}
    Value {from=MaterialExpressionStaticBoolParameter_1, output=""}
   ]
   properties {}
  }
  MaterialExpressionStaticSwitch_2 {
   class=MaterialExpressionStaticSwitch
   inputs [
    True {from=MaterialExpressionCameraVectorWS_6, output=""}
    False {from=MaterialExpressionTransform_7, output=""}
    Value {from=MaterialExpressionStaticBool_1, output=""}
   ]
   properties {}
  }
  MaterialExpressionSubtract_18 {
   class=MaterialExpressionSubtract
   inputs [
    A {from=MaterialExpressionScalarParameter_36}
    B {from=MaterialExpressionScalarParameter_52}
   ]
  }
  MaterialExpressionSubtract_2 {
   class=MaterialExpressionSubtract
   inputs [
    A {from=MaterialExpressionComponentMask_1}
    B {from=""}
   ]
   properties {constB=0.5}
  }
  MaterialExpressionSubtract_20 {
   class=MaterialExpressionSubtract
   inputs [
    A {from=MaterialExpressionConstant_30}
    B {from=MaterialExpressionScalarParameter_52}
   ]
  }
  MaterialExpressionTextureCoordinate_0 {class=MaterialExpressionTextureCoordinate}
  MaterialExpressionTextureCoordinate_1 {class=MaterialExpressionTextureCoordinate}
  MaterialExpressionTextureCoordinate_2 {class=MaterialExpressionTextureCoordinate}
  MaterialExpressionTextureCoordinate_3 {class=MaterialExpressionTextureCoordinate}
  MaterialExpressionTextureCoordinate_4 {class=MaterialExpressionTextureCoordinate}
  MaterialExpressionTextureCoordinate_5 {class=MaterialExpressionTextureCoordinate}
  MaterialExpressionTextureCoordinate_6 {class=MaterialExpressionTextureCoordinate}
  MaterialExpressionTextureCoordinate_7 {class=MaterialExpressionTextureCoordinate}
  MaterialExpressionTextureCoordinate_8 {class=MaterialExpressionTextureCoordinate}
  MaterialExpressionTextureSampleParameter2D_0 {
   class=MaterialExpressionTextureSampleParameter2D
   inputs [
    UVs {from=MaterialExpressionStaticSwitch_0, output=""}
    Apply View MipBias {from="", output=""}
   ]
   properties {group=Texture 1, parameterName=Texture 1, texture=/Game/Effects/Fort_Effects/Textures/Tile/T_Inky_Smoke_Tile2}
  }
  MaterialExpressionTextureSampleParameter2D_1 {
   class=MaterialExpressionTextureSampleParameter2D
   inputs [
    UVs {from=MaterialExpressionPanner_2, output=""}
    Apply View MipBias {from="", output=""}
   ]
   properties {group=Texture Distortion, parameterName=Texture Distortion - Texture, texture=/Game/Effects/Fort_Effects/Textures/Tile/T_PerlinRGB_01_M}
  }
  MaterialExpressionTextureSampleParameter2D_2 {
   class=MaterialExpressionTextureSampleParameter2D
   inputs [
    UVs {from=MaterialExpressionPanner_0, output=""}
    Apply View MipBias {from="", output=""}
   ]
   properties {group=WPO, parameterName=WPO Noise, samplerType=SAMPLERTYPE_LinearColor, texture=/Game/Effects/SnowSandInteraction/Textures/SSI_Perlin}
  }
  MaterialExpressionTextureSampleParameter2D_3 {
   class=MaterialExpressionTextureSampleParameter2D
   inputs [
    UVs {from=MaterialExpressionStaticSwitch_1, output=""}
    Apply View MipBias {from="", output=""}
   ]
   properties {group=Texture 2, parameterName=Texture 2, texture=/Game/Effects/Fort_Effects/Textures/Tile/T_GaussianNoise_01_M}
  }
  MaterialExpressionTextureSampleParameter2D_4 {
   class=MaterialExpressionTextureSampleParameter2D
   inputs [
    UVs {from=MaterialExpressionPanner_1, output=""}
    Apply View MipBias {from="", output=""}
   ]
   properties {group=Dissolve, parameterName=Dissolve - Texture, sortPriority=4, texture=&T_GaussianNoise_01_M}
  }
  MaterialExpressionTextureSampleParameter2D_5 {
   class=MaterialExpressionTextureSampleParameter2D
   inputs [
    UVs {from=MaterialExpressionMultiply_7, output=""}
    Apply View MipBias {from="", output=""}
   ]
   properties {group=Color, parameterName=LerpColorNoise, sortPriority=12, texture=/DealershipSharedLibrary/FX/Textures/Noise/T_VCVFX_NoiseText_A}
  }
  MaterialExpressionTextureSampleParameter2D_6 {
   class=MaterialExpressionTextureSampleParameter2D
   inputs [
    UVs {from=MaterialExpressionPanner_5, output=""}
    Apply View MipBias {from="", output=""}
   ]
   properties {group=Texture 3, parameterName=Texture3, texture=/DealershipSharedLibrary/FX/Textures/Diffuse/T_VCVFX_SpaceTear_D}
  }
  MaterialExpressionTransform_7 {
   class=MaterialExpressionTransform
   inputs [None {from=MaterialExpressionConstant3Vector_9}]
   properties {transformSourceType=TRANSFORMSOURCE_View}
  }
  MaterialExpressionVectorParameter_0 {class=MaterialExpressionVectorParameter, properties {group=Color, parameterName=Color, sortPriority=4, defaultValue=<rgba 1 1 1 1>}}
  MaterialExpressionVectorParameter_1 {
   class=MaterialExpressionVectorParameter
   properties {
    group=WPO
    parameterName="WPO -  Noise UV Controls"
    channelNames {a=Panner V, b=Panner U, g=Tiling V, r=Tiling U}
    defaultValue=<rgba 1 1 0 0>
   }
  }
  MaterialExpressionVectorParameter_10 {class=MaterialExpressionVectorParameter, properties {group=Dissolve, parameterName=Dissolve Highlight Color, defaultValue=<rgba 1 1 1 1>}}
  MaterialExpressionVectorParameter_2 {
   class=MaterialExpressionVectorParameter
   properties {
    group=Dissolve
    parameterName=DissolveUVControls
    sortPriority=8
    channelNames {a=SpeedY, b=SpeedX, g=ScaleY, r=ScaleX}
    defaultValue=<rgba 1 1 0 0>
   }
  }
  MaterialExpressionVectorParameter_3 {
   class=MaterialExpressionVectorParameter
   properties {group=Color, parameterName=LerpColorA, sortPriority=22, defaultValue=<rgba 0.533 0 0.983 0>}
  }
  MaterialExpressionVectorParameter_4 {
   class=MaterialExpressionVectorParameter
   properties {group=Texture 3, parameterName=Texture3Uvs, channelNames {a=SpeedY, b=SpeedX, g=ScaleY, r=ScaleX}, defaultValue=<rgba 1 1 0 0>}
  }
  MaterialExpressionVectorParameter_5 {
   class=MaterialExpressionVectorParameter
   properties {
    group=Texture Distortion
    parameterName=Texture Distortion UV Controls
    sortPriority=5
    channelNames {a=Panner V, b=Panner U, g=Tiling V, r=Tiling U}
    defaultValue=<rgba 1 1 0 0>
   }
  }
  MaterialExpressionVectorParameter_6 {
   class=MaterialExpressionVectorParameter
   properties {group=Color, parameterName=LerpColorB, sortPriority=24, defaultValue=<rgba 0.199 0.231 1 0>}
  }
  MaterialExpressionVectorParameter_7 {
   class=MaterialExpressionVectorParameter
   properties {group=Color, parameterName=LerpColorNoiseUVs, sortPriority=16, channelNames {a="", b="", g=TileY, r=TileX}, defaultValue=<rgba 5 4 0 0>}
  }
  MaterialExpressionVectorParameter_8 {
   class=MaterialExpressionVectorParameter
   properties {
    group=Fresnel
    parameterName=Fresnel - Edge
    channelNames {a=Fresnel Bias, b=Fresnel Amount, g=BaseReflectFractionIn, r=ExponentIn}
    defaultValue=<rgba 5 0.04 1 1>
   }
  }
  MaterialExpressionVectorParameter_9 {
   class=MaterialExpressionVectorParameter
   properties {
    group=Fresnel
    parameterName=Fresnel - Interior
    channelNames {a=Fresnel Bias, b=Fresnel Amount, g=BaseReflectFractionIn, r=ExponentIn}
    defaultValue=<rgba 1 0.75 1 1>
   }
  }
  MaterialExpressionVertexNormalWS_0 {class=MaterialExpressionVertexNormalWS, inputs [], properties {}}
  MaterialExpressionVertexNormalWS_6 {class=MaterialExpressionVertexNormalWS, inputs [], properties {}}
 }
 properties {
  bAutomaticallySetUsageInEditor=false
  bUsedWithNiagaraMeshParticles=true
  blendMode=BLEND_Translucent
  refractionMethod=RM_IndexOfRefraction
  shadingModel=MSM_Unlit
  usedShadingModels=Unlit
 }
}
