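# Unlit translucent master material for Niagara mesh particles, blending a solid color with a Fresnel color, tinted by particle color and passed through MF_FX_EmissiveAdjustments.
# A scalar switch selects plain Fresnel or a panning, smooth-stepped noise texture variant.
# Opacity combines particle alpha, optional invertible Fresnel, a panning mask and a dynamic-parameter multiplier.
# Time-driven sine and cosine waves along vertex normals form vertex displacement, scaled by dynamic WPOAmount.
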Material /DealershipSharedLibrary/FX/Materials/MasterMaterials/Mesh/M_DLR_Mesh_Simple_SolidColor_Fresnel
graph {
 properties {
  bAutomaticallySetUsageInEditor=false
  bUsedWithNiagaraMeshParticles=true
  blendMode=BLEND_Translucent
  shadingModel=MSM_Unlit
  usedShadingModels=Unlit
 }
 nodes [
  n1 = VectorParameter() {parameterName=Fresnel Color, defaultValue=<rgba 1 1 1 1>}
  n2 = Multiply(A=@n1.RGB, B=NamedRerouteUsage())
  n3 = NamedRerouteUsage()
  n4 = If(
   A=ScalarParameter() {parameterName=Use Texture?}
   "A > B"=@n2
   "A == B"=@n2
   "A < B"=LinearInterpolate(A=VectorParameter() {parameterName=Solid Color}.RGB, B=Multiply(A=@n1.RGB, B=@n3), Alpha=NamedRerouteUsage())
  ) {constB=1}
  n5 = ScalarParameter() {defaultValue=1, parameterName="Use Fresnel 1=True 0=Flase"}
  n6 = ParticleColor()
  MaterialFunctionCall(Emissive=Multiply(A=LinearInterpolate(B=@n4, Alpha=@n5), B=@n6.RGB)) {materialFunction=&MF_FX_EmissiveAdjustments}
  n7 = ScalarParameter() {parameterName=UseOpacityMask}
  n8 = VectorParameter() {parameterName=Opacity Mask UV Controls, channelNames {a=Panner Y, b=Panner X, g=Tile Y, r=Tile X}, defaultValue=<rgba 1 1 0 0>}
  n9 = Multiply(A=TextureCoordinate(), B=AppendVector(A=@n8.Tile X, B=@n8.Tile X))
  n10 = AppendVector(A=@n8.Panner X, B=@n8.Panner X)
  n11 = Fresnel(
   ExponentIn=ScalarParameter() {defaultValue=5, parameterName=FresnelOpacity}
   BaseReflectFractionIn=ScalarParameter() {parameterName=FresnelOpacity - Thickness}
  ) {baseReflectFraction=0}
  n12 = ScalarParameter() {defaultValue=1, parameterName=FresnelOpacity - Power}
  n13 = Saturate(None=Multiply(A=Power(Base=@n11, Exp=@n12), B=ScalarParameter() {defaultValue=1, parameterName=FresnelOpacity - Mult}))
  n14 = OneMinus(None=@n13)
  n15 = If(A=ScalarParameter() {parameterName=FresnelOpacity - Invert}, "A > B"=@n14, "A == B"=@n14, "A < B"=@n13) {constB=1}
  n16 = If(A=ScalarParameter() {parameterName=UseFresnelForAlpha?}, "A > B"=@n15, "A == B"=@n15, "A < B"=Constant() {r=1}) {constB=1}
  n17 = Multiply(
   A=@n16
   B=TextureSampleParameter2D(UVs=Panner(Coordinate=@n9, Speed=@n10)) {parameterName=OpactityMask, texture=/Game/Effects/Fort_Effects/Textures/Mask/Mask_T_Mask_SoftSquare}.RGB
  )
  n18 = DynamicParameter() {parameterIndex=1, defaultValue=<rgba 1 1 1 0>, paramNames [OpacityMultiplier, Param2, Param3, Param4]}
  n19 = Constant() {r=1}
  Multiply(
   A=Multiply(A=@n6.A, B=If(A=@n7, "A > B"=@n17, "A == B"=@n17, "A < B"=@n16) {constB=1})
   B=If(A=ScalarParameter() {parameterName=UseDynamicParamOpacityMultiplier}, "A > B"=@n18.OpacityMultiplier, "A == B"=@n19, "A < B"=@n19)
  )
  n20 = DynamicParameter() {defaultValue=<rgba 1 1 1 0>, paramNames [Fresnel Strength, Fresnel Thickness, Fresnel Power, WPOAmount]}
  n21 = Fresnel(ExponentIn=@n20.Fresnel Strength, BaseReflectFractionIn=@n20.Fresnel Strength) {baseReflectFraction=0.5}
  n22 = Power(Base=LinearInterpolate(A=@n21, B=OneMinus(None=@n21), Alpha=ScalarParameter() {parameterName=Invert Fresnel}), Exp=@n20.Fresnel Power)
  NamedRerouteDeclaration(None=@n22) {name=Fresnel, nodeColor=<rgba 0.3359 1 0 1>}
  n23 = ScalarParameter() {defaultValue=0.2, parameterName=TextureStepMin}
  n24 = Add(A=@n23, B=ScalarParameter() {parameterName=TextureSmoothness})
  n25 = VectorParameter() {parameterName=Fresnel Texture Controls, channelNames {a=SpeedY, b=SpeedX, g=TilingY, r=TilingX}, defaultValue=<rgba 1 1 0 0>}
  n26 = Panner(Coordinate=Multiply(A=TextureCoordinate(), B=AppendVector(A=@n25.TilingX, B=@n25.TilingX)), Speed=AppendVector(A=@n25.SpeedX, B=@n25.SpeedX))
  n27 = Multiply(
   A=Power(
    Base=TextureSampleParameter2D(UVs=@n26) {parameterName=Fresnel Texture, texture=/DealershipSharedLibrary/FX/Textures/Noise/T_VCVFX_CellNoiseSkins_A}.RGB
    Exp=ScalarParameter() {defaultValue=1, parameterName=FresnelTexturePower}
   )
   B=@n22
  )
  n28 = ScalarParameter() {parameterName=BackgroundFresnelExponent}
  NamedRerouteDeclaration(None=Add(A=SmoothStep(Min=@n23, Max=@n24, Value=@n27), B=Fresnel(ExponentIn=@n28) {baseReflectFraction=0})) {name=TextureOut, nodeColor=<rgba 1 0 0.6562 1>}
  n29 = Add(
   A=Multiply(A=Time(), B=ScalarParameter() {defaultValue=1, parameterName=WPOTimeMultiplierZ})
   B=ComponentMask(None=TextureCoordinate()) {g=true}
  )
  n30 = Cosine(None=@n29) {period=0.8}
  n31 = AppendVector(A=Sine(None=@n29) {period=1.5}, B=@n30)
  n32 = Multiply(A=ConstantBiasScale(None=AppendVector(A=@n31, B=@n30)) {scale=2}, B=VertexNormalWS())
  n33 = Add(
   A=Multiply(A=Time(), B=ScalarParameter() {defaultValue=1, parameterName=WPOTimeMultiplierXY})
   B=ComponentMask(None=TextureCoordinate()) {r=true}
  )
  n34 = Cosine(None=@n33)
  n35 = AppendVector(A=AppendVector(A=Sine(None=@n33) {period=2}, B=@n34), B=@n34)
  n36 = Multiply(A=Multiply(A=ConstantBiasScale(None=@n35) {scale=1.5}, B=VertexNormalWS()), B=ScalarParameter() {parameterName=WPO_XYMultipler})
  n37 = DynamicParameter() {defaultValue=<rgba 1 1 1 0>, paramNames [Fresnel Strength, Fresnel Thickness, Fresnel Power, WPOAmount]}
  Reroute(None=Multiply(A=Add(A=@n32, B=@n36), B=@n37.WPOAmount))
 ]
}
