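# Unlit, masked, two-sided FX master material that draws animated swirl streaks on sphere meshes, usable with Niagara.
# A power-sharpened band around the middle of UV V multiplies a sine along U to form the streak mask.
# Dynamic parameters drive streak width and offset, with static debug switches substituting scalar and time-based values.
# Emission combines the mask, inside/outside colors by face side and particle color, then passes through MF_FX_EmissiveAdjustments.
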Material /DealershipSharedLibrary/FX/Materials/MasterMaterials/Mesh/M_DLR_Mesh_SphereSwirls
graph {
 properties {
  bUsedWithNanite=true
  bUsedWithNiagaraMeshParticles=true
  bUsedWithNiagaraSprites=true
  blendMode=BLEND_Masked
  pixelDepthOffsetMode=PDOM_Legacy
  shadingModel=MSM_Unlit
  twoSided=true
  usedShadingModels=Unlit
 }
 nodes [
  Constant() {r=1}
  n1 = TextureCoordinate()
  n2 = Abs(None=ComponentMask(None=Add(A=@n1) {constB=-0.5}) {g=true})
  n3 = MaterialFunctionCall(
   Input=ScalarParameter() {defaultValue=15, parameterName=Width, sliderMax=5, sliderMin=1}
   Input Low=Constant() {r=1}
   Input High=Constant() {r=5}
   Target Low=Constant() {r=2.8}
   Target High=Constant() {r=30}
  ) {materialFunction=&RemapValueRange}
  n4 = DynamicParameter() {defaultValue=<rgba 15 0 1 1>, paramNames [StreakWidth,StreakOffset,Param3,Param4]}
  n5 = StaticBoolParameter() {parameterName=DebugWidth}
  n6 = ScalarParameter() {defaultValue=0.5, parameterName=StreaksLength, sliderMax=0.5, sliderMin=0.145}
  n7 = AppendVector(A=@n6, B=Constant() {r=1})
  n8 = ComponentMask(None=Frac(None=Multiply(A=@n1, B=@n7))) {r=true}
  n9 = StaticSwitch(
   True=Multiply(A=ScalarParameter() {parameterName=StreakOffset}, B=Time())
   False=@n4.StreakOffset
   Value=StaticBoolParameter() {parameterName=DebugOffset}
  )
  n10 = ConstantBiasScale(None=Sine(None=Add(A=@n8, B=@n9) {constB=0}) {period=0.5}) {bias=-0.05, scale=0.8}
  n11 = Multiply(A=Power(Base=OneMinus(None=@n2), Exp=StaticSwitch(True=@n3.Result, False=@n4.StreakWidth, Value=@n5)) {constExponent=1}, B=@n10)
  n12 = ScalarParameter() {defaultValue=1, parameterName=EmissionPower}
  n13 = Multiply(
   A=Saturate(None=Power(Base=@n11, Exp=@n12) {constExponent=1.5})
   B=LinearInterpolate(
    A=VectorParameter() {parameterName=InsideColor, defaultValue=<rgba 1 1 1 1>}.RGB
    B=VectorParameter() {parameterName=OutsideColor}.RGB
    Alpha=TwoSidedSign()
   )
  )
  MaterialFunctionCall(Emissive=Multiply(A=ParticleColor().RGB, B=StaticSwitchParameter(True=@n11, False=@n13) {parameterName=DebugMask})) {materialFunction=&MF_FX_EmissiveAdjustments}
  MaterialFunctionCall(Input=@n6, Input Low=Constant() {r=1}, Input High=Constant(), Target Low=Constant() {r=0.145}, Target High=Constant() {r=0.5}) {materialFunction=&RemapValueRange}
  n14 = Saturate(None=ScalarParameter() {defaultValue=1, parameterName=Levels - Max})
  n15 = Min(A=Saturate(None=@n4.StreakWidth), B=@n14)
  n16 = Clamp(None=@n10)
  Saturate(None=Divide(A=Subtract(A=@n16, B=@n15), B=Subtract(A=@n14, B=@n15)))
  ScalarParameter() {defaultValue=15, parameterName=Width}
  StaticSwitchParameter(True=Constant() {r=1}, False=@n11) {parameterName=DebugMask}
 ]
}
